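# Physics asset for the Veiled female character's face accessory, providing bodies and joints for its secondary motion on dyn_hair bones.
# Short capsule chains A and C on the right and E on both sides hang from small MSTR root bodies jointed to the head.
# Links within chains have limited swing and twist, while root and body joints rotate freely.
# Head, neck, upper spine, right clavicle and right upper-arm capsules approximate the surrounding body.
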PhysicsAsset /BRCosmetics/Characters/Player/Female/Medium/Bodies/F_MED_Veiled/Meshes/Parts/F_MED_Veiled_FaceAcc_Physics
bodies {
 clavicle_r [{
   shapeType=Capsule
   length=5.9909
   radius=5.2737
   shapeName=clavicle_r_capsule
   center=<-6.4341 0.3237 0.8599>
   rotation=<rot 3.2384 98.0696 -90>
  }]
 dyn_hair_A_01_r [{shapeType=Capsule, length=2, radius=2, shapeName=dyn_hair_A_01_r_capsule, center=<2.5 0 0>, rotation=<rot 90 0 0>}]
 dyn_hair_A_02_r [
  {shapeType=Capsule, length=2, radius=2.25, shapeName=dyn_hair_A_02_r_capsule, center=<2.5 0 -0.1353>, rotation=<rot 90 0 0>}
  {shapeType=Capsule, length=3.25, radius=1.25, shapeName=dyn_hair_A_02_r_capsule, center=<2.6059 -2.5387 0.6096>, rotation=<rot 90 0 0>}
  {shapeType=Capsule, length=3.25, radius=2, shapeName=dyn_hair_A_02_r_capsule, center=<-0.6205 2.31 -1.8293>, rotation=<rot 90 0 0>}
 ]
 dyn_hair_A_03_r [{shapeType=Capsule, length=2, radius=2, shapeName=dyn_hair_A_03_r_capsule, center=<2.5 0 0>, rotation=<rot 90 0 0>}]
 dyn_hair_A_r_MSTR [{shapeType=Capsule, length=0.505, radius=0.505, shapeName=dyn_hair_A_r_MSTR_capsule, rotation=<rot 0 -90 0>}]
 dyn_hair_C_01_r [{shapeType=Capsule, length=2.38, radius=1.255, shapeName=dyn_hair_C_01_r_capsule, center=<2 0 0>, rotation=<rot 90 0 0>}]
 dyn_hair_C_02_r [{shapeType=Capsule, length=2.38, radius=1.255, shapeName=dyn_hair_C_02_r_capsule, center=<2 0 0>, rotation=<rot 90 0 0>}]
 dyn_hair_C_r_MSTR [{shapeType=Capsule, length=0.505, radius=0.505, shapeName=dyn_hair_C_r_MSTR_capsule, rotation=<rot 0 -90 0>}]
 dyn_hair_E_01_l [{shapeType=Capsule, length=1.63, radius=1, shapeName=dyn_hair_E_01_l_capsule, center=<2 0 1>, rotation=<rot 90 0 0>}]
 dyn_hair_E_01_r [{shapeType=Capsule, length=1.63, radius=1, shapeName=dyn_hair_E_01_r_capsule, center=<1.5 0 0>, rotation=<rot 90 0 0>}]
 dyn_hair_E_02_r [{shapeType=Capsule, length=2.8163, radius=1.3566, shapeName=dyn_hair_E_02_r_capsule, center=<2 0 0>, rotation=<rot 90 0 0>}]
 dyn_hair_E_l_MSTR [{shapeType=Capsule, length=0.505, radius=0.505, shapeName=dyn_hair_E_l_MSTR_capsule, rotation=<rot 0 -90 0>}]
 dyn_hair_E_r_MSTR [{shapeType=Capsule, length=0.505, radius=0.505, shapeName=dyn_hair_E_r_MSTR_capsule, rotation=<rot 0 -90 0>}]
 head [
  {shapeType=Capsule, length=6.0239, radius=5.25, center=<0.8793 1.2376 0.4523>, rotation=<rot 90 0 0>}
  {shapeType=Capsule, length=6.25, radius=4.5, shapeName=head_capsule, center=<0 4.1906 -0.891>, rotation=<rot 90 0 0>}
 ]
 neck_01 [{shapeType=Capsule, length=1.75, radius=4.6867, shapeName=head_capsule, center=<1.4786 0.9542 0>, rotation=<rot 90 0 0>}]
 spine_04 [{shapeType=Capsule, length=4.4066, radius=10.2626, shapeName=spine_04_capsule, center=<0.306 0.5327 0>, rotation=<rot 0 -92.6847 0>}]
 spine_05 [{shapeType=Capsule, length=12.9411, radius=8.1451, shapeName=spine_05_capsule, center=<0.357 -2.1578 -0.1551>}]
 upperarm_r [{shapeType=Capsule, length=13.4229, radius=4.1429, shapeName=upperarm_r_capsule, center=<-5 1.25 0.0001>, rotation=<rot 0 -90 90>}]
}
constraints [
 {bone1Name=dyn_hair_A_03_r, bone2Name=dyn_hair_A_02_r, swing1LimitDegrees=25, swing2LimitDegrees=15, twistLimitDegrees=15}
 {
  bone1Name=dyn_hair_A_r_MSTR
  bone2Name=head
  swing1LimitDegrees=45
  swing1Motion=Free
  swing2LimitDegrees=45
  swing2Motion=Free
  twistLimitDegrees=45
  twistMotion=Free
 }
 {bone1Name=dyn_hair_A_01_r, bone2Name=dyn_hair_A_r_MSTR, swing1LimitDegrees=25, swing2LimitDegrees=15, twistLimitDegrees=15}
 {bone1Name=dyn_hair_A_02_r, bone2Name=dyn_hair_A_01_r, swing1LimitDegrees=25, swing2LimitDegrees=15, twistLimitDegrees=15}
 {
  bone1Name=head
  bone2Name=neck_01
  swing1LimitDegrees=45
  swing1Motion=Free
  swing2LimitDegrees=45
  swing2Motion=Free
  twistLimitDegrees=45
  twistMotion=Free
 }
 {
  bone1Name=spine_05
  bone2Name=spine_04
  swing1LimitDegrees=45
  swing1Motion=Free
  swing2LimitDegrees=45
  swing2Motion=Free
  twistLimitDegrees=45
  twistMotion=Free
 }
 {
  bone1Name=clavicle_r
  bone2Name=spine_05
  swing1LimitDegrees=45
  swing1Motion=Free
  swing2LimitDegrees=45
  swing2Motion=Free
  twistLimitDegrees=45
  twistMotion=Free
 }
 {
  bone1Name=upperarm_r
  bone2Name=clavicle_r
  swing1LimitDegrees=45
  swing1Motion=Free
  swing2LimitDegrees=45
  swing2Motion=Free
  twistLimitDegrees=45
  twistMotion=Free
 }
 {
  bone1Name=dyn_hair_E_l_MSTR
  bone2Name=head
  swing1LimitDegrees=45
  swing1Motion=Free
  swing2LimitDegrees=45
  swing2Motion=Free
  twistLimitDegrees=45
  twistMotion=Free
 }
 {bone1Name=dyn_hair_E_01_l, bone2Name=dyn_hair_E_l_MSTR, swing1LimitDegrees=45, swing2LimitDegrees=45, twistLimitDegrees=3}
 {
  bone1Name=dyn_hair_E_r_MSTR
  bone2Name=head
  swing1LimitDegrees=45
  swing1Motion=Free
  swing2LimitDegrees=45
  swing2Motion=Free
  twistLimitDegrees=45
  twistMotion=Free
 }
 {bone1Name=dyn_hair_E_01_r, bone2Name=dyn_hair_E_r_MSTR, swing1LimitDegrees=45, swing2LimitDegrees=45, twistLimitDegrees=3}
 {bone1Name=dyn_hair_E_02_r, bone2Name=dyn_hair_E_01_r, swing1LimitDegrees=45, swing2LimitDegrees=45, twistLimitDegrees=3}
 {
  bone1Name=neck_01
  bone2Name=spine_05
  swing1LimitDegrees=45
  swing1Motion=Free
  swing2LimitDegrees=45
  swing2Motion=Free
  twistLimitDegrees=45
  twistMotion=Free
 }
 {
  bone1Name=dyn_hair_C_r_MSTR
  bone2Name=head
  swing1LimitDegrees=45
  swing1Motion=Free
  swing2LimitDegrees=45
  swing2Motion=Free
  twistLimitDegrees=45
  twistMotion=Free
 }
 {bone1Name=dyn_hair_C_01_r, bone2Name=dyn_hair_C_r_MSTR, swing1LimitDegrees=45, swing2LimitDegrees=45, twistLimitDegrees=15}
 {bone1Name=dyn_hair_C_02_r, bone2Name=dyn_hair_C_01_r, swing1LimitDegrees=45, swing2LimitDegrees=45, twistLimitDegrees=15}
]
