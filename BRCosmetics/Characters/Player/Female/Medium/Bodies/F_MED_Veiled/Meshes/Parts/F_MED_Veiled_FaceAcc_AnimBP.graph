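# Anim Blueprint for the Veiled female character's face accessory mesh, adding secondary motion to bun and hair bones.
# It layers two copied mesh poses at the upper arms, runs chained AnimDynamics, then caches the result.
# bIsInFrontEnd blends between world-space front-end and base-bone-space RigidBody simulations, with gravity updated by a deprecated state function.
# The dyn_hair_C right-side AnimDynamics nodes are disconnected from the output.
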AnimBlueprint /BRCosmetics/Characters/Player/Female/Medium/Bodies/F_MED_Veiled/Meshes/Parts/F_MED_Veiled_FaceAcc_AnimBP
target_skeleton=&/F_MED_Veiled_FaceAcc_Skeleton
graphs [
 AnimGraph {nodes [
   n1 = Anim Dynamics - Bone: dyn_hair_C_01_r(Alpha=1, ExternalForce="0,0,0") {
    angularDampingOverride=0.9999
    angularSpringConstant=100
    bAngularSpring=true
    bGravityOverrideInSimSpace=true
    bOverrideAngularDamping=true
    simulationSpace=RootRelative
    alphaBoolBlend {blendOption=Linear}
    boundBone {boneName=dyn_hair_C_01_r}
    componentAppliedLinearAccClamp=<500 500 500>
    componentLinearAccScale=<0.5 0.5 0.5>
    componentLinearVelScale=<0.5 0.5 0.5>
    physicsBodyDefinitions [{
      sphereCollisionRadius=0
      boundBone {boneName=dyn_hair_C_01_r}
      boxExtents=<15 10 10>
      constraintSetup {angularLimitsMax=<1 0 15>, angularLimitsMin=<-1 -30 -5>, angularTarget=<1 0 0>}
      localJointOffset=<2 0 0>
     }]
    relativeSpaceBone {boneName=root}
   }
   Anim Dynamics - Bone: dyn_hair_C_02_r(ComponentPose=@n1, Alpha=1, ExternalForce="0,0,0") {
    angularDampingOverride=0.9999
    angularSpringConstant=100
    bAngularSpring=true
    bGravityOverrideInSimSpace=true
    bOverrideAngularDamping=true
    simulationSpace=RootRelative
    alphaBoolBlend {blendOption=Linear}
    boundBone {boneName=dyn_hair_C_02_r}
    componentAppliedLinearAccClamp=<500 500 500>
    componentLinearAccScale=<0.5 0.5 0.5>
    componentLinearVelScale=<0.5 0.5 0.5>
    physicsBodyDefinitions [{
      sphereCollisionRadius=0
      boundBone {boneName=dyn_hair_C_02_r}
      boxExtents=<15 10 10>
      constraintSetup {angularLimitsMax=<1 0 15>, angularLimitsMin=<-1 -30 -15>, angularTarget=<1 0 0>}
      localJointOffset=<2 0 0>
     }]
    relativeSpaceBone {boneName=root}
   }
   n2 = Get GravityOverride()
   n3 = RigidBody!(ComponentPose=Local To Component(LocalPose=Use cached pose 'animDynamics_Cache'()), Alpha, OverrideWorldGravity:Vector=@n2, ExternalForce) -> (Pose) {
    bClampLinearTranslationLimitToRefPose=true
    bOverrideWorldGravity=true
    overridePhysicsAsset=&/F_MED_Veiled_FaceAcc_FrontEnd_Physics
    simulationSpace=WorldSpace
    alphaBoolBlend {blendOption=Linear}
    baseBoneRef {boneName=root}
    componentLinearAccScale=<0.3 0.3 0.4>
    componentLinearVelScale=<0.3 0.3 0.4>
    overrideWorldGravity {z=-980}
   }
   n4 = RigidBody(ComponentPose=Local To Component(LocalPose=Use cached pose 'animDynamics_Cache'()), OverrideWorldGravity=@n2) {
    bClampLinearTranslationLimitToRefPose=true
    bOverrideWorldGravity=true
    overridePhysicsAsset=&/F_MED_Veiled_FaceAcc_Physics
    simulationSpace=BaseBoneSpace
    alphaBoolBlend {blendOption=Linear}
    baseBoneRef {boneName=root}
    componentLinearAccScale=<0.3 0.3 0.4>
    componentLinearVelScale=<0.3 0.3 0.4>
    overrideWorldGravity {z=-980}
   }
   Output Pose(Result=Blend Poses by bool(BlendPose_0=Component To Local(ComponentPose=@n3), BlendPose_1=Component To Local(ComponentPose=@n4), bActiveValue=Get bIsInFrontEnd()) {blendType=Linear}) {layerGroup=None}
   n5 = Copy Pose From Mesh(SourceMeshComponent=Get Skel_Mesh()) {bCopyCurves=true, bUseAttachedParent=true}
   n6 = Copy Pose From Mesh() {bCopyCurves=true, bUseAttachedParent=true}
   n7 = Anim Dynamics - Bone: dyn_bun(
    ComponentPose=Local To Component(LocalPose=Layered blend per bone(BasePose=@n6, BlendPoses_0=@n5) {layerSetup [{branchFilters [{boneName=upperarm_l}, {boneName=upperarm_r}]}]})
    Alpha=0.7
    ExternalForce="0,0,0"
   ) {
    angularDampingOverride=1
    angularSpringConstant=1000
    bAngularSpring=true
    bOverrideAngularDamping=true
    bUseGravityOverride=true
    simulationSpace=RootRelative
    alphaBoolBlend {blendOption=Linear}
    boundBone {boneName=dyn_bun}
    componentAppliedLinearAccClamp=<2000 2000 2000>
    componentLinearAccScale=<0.3 0.3 0.3>
    componentLinearVelScale=<0.3 0.3 0.3>
    gravityOverride {z=-980}
    physicsBodyDefinitions [{
      boundBone {boneName=dyn_bun}
      boxExtents=<10 20 20>
      constraintSetup {angularLimitsMax=<5 5 5>, angularLimitsMin=<-5 -5 -5>, angularTarget=<0 0 0>}
      localJointOffset=<4 0 0>
     }]
   }
   n8 = Anim Dynamics - Bone: dyn_hair_A_01_l(ComponentPose=@n7, Alpha=1, ExternalForce="0,0,0") {
    angularDampingOverride=0.9999
    angularSpringConstant=100
    bAngularSpring=true
    bGravityOverrideInSimSpace=true
    bOverrideAngularDamping=true
    simulationSpace=Actor
    alphaBoolBlend {blendOption=Linear}
    boundBone {boneName=dyn_hair_A_01_l}
    componentAppliedLinearAccClamp=<500 500 500>
    componentLinearAccScale=<0.5 0.5 0.5>
    componentLinearVelScale=<0.5 0.5 0.5>
    physicsBodyDefinitions [{
      sphereCollisionRadius=0
      boundBone {boneName=dyn_hair_A_01_l}
      boxExtents=<15 10 10>
      constraintSetup {angularLimitsMax=<1 10 15>, angularLimitsMin=<-1 -30 -15>, angularTarget=<1 0 0>}
      localJointOffset=<2 0 0>
     }]
    relativeSpaceBone {boneName=root}
   }
   n9 = Anim Dynamics - Bone: dyn_hair_A_02_l(ComponentPose=@n8, Alpha=1, ExternalForce="0,0,0") {
    angularDampingOverride=0.9999
    angularSpringConstant=100
    bAngularSpring=true
    bGravityOverrideInSimSpace=true
    bOverrideAngularDamping=true
    simulationSpace=Actor
    alphaBoolBlend {blendOption=Linear}
    boundBone {boneName=dyn_hair_A_02_l}
    componentAppliedLinearAccClamp=<500 500 500>
    componentLinearAccScale=<0.5 0.5 0.5>
    componentLinearVelScale=<0.5 0.5 0.5>
    physicsBodyDefinitions [{
      sphereCollisionRadius=0
      boundBone {boneName=dyn_hair_A_02_l}
      boxExtents=<15 10 10>
      constraintSetup {angularLimitsMax=<1 10 15>, angularLimitsMin=<-1 -30 -15>, angularTarget=<1 0 0>}
      localJointOffset=<2 0 0>
     }]
    relativeSpaceBone {boneName=root}
   }
   n10 = Anim Dynamics - Bone: dyn_hair_B_01_l(ComponentPose=@n9, Alpha=1, ExternalForce="0,0,0") {
    angularDampingOverride=0.9999
    angularSpringConstant=100
    bAngularSpring=true
    bGravityOverrideInSimSpace=true
    bOverrideAngularDamping=true
    simulationSpace=RootRelative
    alphaBoolBlend {blendOption=Linear}
    boundBone {boneName=dyn_hair_B_01_l}
    componentAppliedLinearAccClamp=<500 500 500>
    componentLinearAccScale=<0.5 0.5 0.5>
    componentLinearVelScale=<0.5 0.5 0.5>
    physicsBodyDefinitions [{
      sphereCollisionRadius=0
      boundBone {boneName=dyn_hair_B_01_l}
      boxExtents=<15 10 10>
      constraintSetup {angularLimitsMax=<1 0 15>, angularLimitsMin=<-1 -30 -5>, angularTarget=<1 0 0>}
      localJointOffset=<2 0 0>
     }]
    relativeSpaceBone {boneName=root}
   }
   n11 = Anim Dynamics - Bone: dyn_hair_B_02_l(ComponentPose=@n10, Alpha=1, ExternalForce="0,0,0") {
    angularDampingOverride=0.9999
    angularSpringConstant=100
    bAngularSpring=true
    bGravityOverrideInSimSpace=true
    bOverrideAngularDamping=true
    simulationSpace=RootRelative
    alphaBoolBlend {blendOption=Linear}
    boundBone {boneName=dyn_hair_B_02_l}
    componentAppliedLinearAccClamp=<500 500 500>
    componentLinearAccScale=<0.5 0.5 0.5>
    componentLinearVelScale=<0.5 0.5 0.5>
    physicsBodyDefinitions [{
      sphereCollisionRadius=0
      boundBone {boneName=dyn_hair_B_02_l}
      boxExtents=<15 10 10>
      constraintSetup {angularLimitsMax=<1 0 15>, angularLimitsMin=<-1 -30 -15>, angularTarget=<1 0 0>}
      localJointOffset=<2 0 0>
     }]
    relativeSpaceBone {boneName=root}
   }
   n12 = Anim Dynamics - Bone: dyn_hair_B_01_r(ComponentPose=@n11, Alpha=1, ExternalForce="0,0,0") {
    angularDampingOverride=0.9999
    angularSpringConstant=100
    bAngularSpring=true
    bGravityOverrideInSimSpace=true
    bOverrideAngularDamping=true
    simulationSpace=Actor
    alphaBoolBlend {blendOption=Linear}
    boundBone {boneName=dyn_hair_B_01_r}
    componentAppliedLinearAccClamp=<500 500 500>
    componentLinearAccScale=<0.5 0.5 0.5>
    componentLinearVelScale=<0.5 0.5 0.5>
    physicsBodyDefinitions [{
      sphereCollisionRadius=0
      boundBone {boneName=dyn_hair_B_01_r}
      boxExtents=<15 10 10>
      constraintSetup {angularLimitsMax=<1 10 15>, angularLimitsMin=<-1 -30 -15>, angularTarget=<1 0 0>}
      localJointOffset=<2 0 0>
     }]
    relativeSpaceBone {boneName=root}
   }
   n13 = Anim Dynamics - Bone: dyn_hair_D_01_r(ComponentPose=@n12, Alpha=1, ExternalForce="0,0,0") {
    angularDampingOverride=0.9999
    angularSpringConstant=100
    bAngularSpring=true
    bGravityOverrideInSimSpace=true
    bOverrideAngularDamping=true
    simulationSpace=Actor
    alphaBoolBlend {blendOption=Linear}
    boundBone {boneName=dyn_hair_D_01_r}
    componentAppliedLinearAccClamp=<500 500 500>
    componentLinearAccScale=<0.5 0.5 0.5>
    componentLinearVelScale=<0.5 0.5 0.5>
    physicsBodyDefinitions [{
      sphereCollisionRadius=0
      boundBone {boneName=dyn_hair_D_01_r}
      boxExtents=<15 10 10>
      constraintSetup {angularLimitsMax=<1 10 15>, angularLimitsMin=<-1 -30 -15>, angularTarget=<1 0 0>}
      localJointOffset=<2 0 0>
     }]
    relativeSpaceBone {boneName=root}
   }
   n14 = Anim Dynamics - Bone: dyn_hair_D_02_r(ComponentPose=@n13, Alpha=1, ExternalForce="0,0,0") {
    angularDampingOverride=0.9999
    angularSpringConstant=100
    bAngularSpring=true
    bGravityOverrideInSimSpace=true
    bOverrideAngularDamping=true
    simulationSpace=Actor
    alphaBoolBlend {blendOption=Linear}
    boundBone {boneName=dyn_hair_D_02_r}
    componentAppliedLinearAccClamp=<500 500 500>
    componentLinearAccScale=<0.5 0.5 0.5>
    componentLinearVelScale=<0.5 0.5 0.5>
    physicsBodyDefinitions [{
      sphereCollisionRadius=0
      boundBone {boneName=dyn_hair_D_02_r}
      boxExtents=<15 10 10>
      constraintSetup {angularLimitsMax=<1 10 15>, angularLimitsMin=<-1 -30 -15>, angularTarget=<1 0 0>}
      localJointOffset=<2 0 0>
     }]
    relativeSpaceBone {boneName=root}
   }
   n15 = Anim Dynamics - Bone: dyn_hair_C_01_l(ComponentPose=@n14, Alpha=1, ExternalForce="0,0,0") {
    angularDampingOverride=0.9999
    angularSpringConstant=100
    bAngularSpring=true
    bGravityOverrideInSimSpace=true
    bOverrideAngularDamping=true
    simulationSpace=RootRelative
    alphaBoolBlend {blendOption=Linear}
    boundBone {boneName=dyn_hair_C_01_l}
    componentAppliedLinearAccClamp=<500 500 500>
    componentLinearAccScale=<0.5 0.5 0.5>
    componentLinearVelScale=<0.5 0.5 0.5>
    physicsBodyDefinitions [{
      sphereCollisionRadius=0
      boundBone {boneName=dyn_hair_C_01_l}
      boxExtents=<15 10 10>
      constraintSetup {angularLimitsMax=<1 0 15>, angularLimitsMin=<-1 -30 -5>, angularTarget=<1 0 0>}
      localJointOffset=<2 0 0>
     }]
    relativeSpaceBone {boneName=root}
   }
   n16 = Anim Dynamics - Bone: dyn_hair_C_02_l(ComponentPose=@n15, Alpha=1, ExternalForce="0,0,0") {
    angularDampingOverride=0.9999
    angularSpringConstant=100
    bAngularSpring=true
    bGravityOverrideInSimSpace=true
    bOverrideAngularDamping=true
    simulationSpace=RootRelative
    alphaBoolBlend {blendOption=Linear}
    boundBone {boneName=dyn_hair_C_02_l}
    componentAppliedLinearAccClamp=<500 500 500>
    componentLinearAccScale=<0.5 0.5 0.5>
    componentLinearVelScale=<0.5 0.5 0.5>
    physicsBodyDefinitions [{
      sphereCollisionRadius=0
      boundBone {boneName=dyn_hair_C_02_l}
      boxExtents=<15 10 10>
      constraintSetup {angularLimitsMax=<1 0 15>, angularLimitsMin=<-1 -30 -15>, angularTarget=<1 0 0>}
      localJointOffset=<2 0 0>
     }]
    relativeSpaceBone {boneName=root}
   }
   n17 = Anim Dynamics - Bone: dyn_hair_D_01_l(ComponentPose=@n16, Alpha=1, ExternalForce="0,0,0") {
    angularDampingOverride=0.9999
    angularSpringConstant=100
    bAngularSpring=true
    bGravityOverrideInSimSpace=true
    bOverrideAngularDamping=true
    simulationSpace=Actor
    alphaBoolBlend {blendOption=Linear}
    boundBone {boneName=dyn_hair_D_01_l}
    componentAppliedLinearAccClamp=<500 500 500>
    componentLinearAccScale=<0.5 0.5 0.5>
    componentLinearVelScale=<0.5 0.5 0.5>
    physicsBodyDefinitions [{
      sphereCollisionRadius=0
      boundBone {boneName=dyn_hair_D_01_l}
      boxExtents=<15 10 10>
      constraintSetup {angularLimitsMax=<1 10 15>, angularLimitsMin=<-1 -30 -15>, angularTarget=<1 0 0>}
      localJointOffset=<2 0 0>
     }]
    relativeSpaceBone {boneName=root}
   }
   n18 = Anim Dynamics - Bone: dyn_hair_D_02_l(ComponentPose=@n17, Alpha=1, ExternalForce="0,0,0") {
    angularDampingOverride=0.9999
    angularSpringConstant=100
    bAngularSpring=true
    bGravityOverrideInSimSpace=true
    bOverrideAngularDamping=true
    simulationSpace=Actor
    alphaBoolBlend {blendOption=Linear}
    boundBone {boneName=dyn_hair_D_02_l}
    componentAppliedLinearAccClamp=<500 500 500>
    componentLinearAccScale=<0.5 0.5 0.5>
    componentLinearVelScale=<0.5 0.5 0.5>
    physicsBodyDefinitions [{
      sphereCollisionRadius=0
      boundBone {boneName=dyn_hair_D_02_l}
      boxExtents=<15 10 10>
      constraintSetup {angularLimitsMax=<1 10 15>, angularLimitsMin=<-1 -30 -15>, angularTarget=<1 0 0>}
      localJointOffset=<2 0 0>
     }]
    relativeSpaceBone {boneName=root}
   }
   n19 = Anim Dynamics - Bone: dyn_hair_E_01_r(ComponentPose=@n18, Alpha=1, ExternalForce="0,0,0") {
    angularDampingOverride=0.9999
    angularSpringConstant=100
    bAngularSpring=true
    bGravityOverrideInSimSpace=true
    bOverrideAngularDamping=true
    simulationSpace=RootRelative
    alphaBoolBlend {blendOption=Linear}
    boundBone {boneName=dyn_hair_E_01_r}
    componentAppliedLinearAccClamp=<500 500 500>
    componentLinearAccScale=<0.5 0.5 0.5>
    componentLinearVelScale=<0.5 0.5 0.5>
    physicsBodyDefinitions [{
      sphereCollisionRadius=0
      boundBone {boneName=dyn_hair_E_01_r}
      boxExtents=<15 10 10>
      constraintSetup {angularLimitsMax=<1 0 15>, angularLimitsMin=<-1 -30 -5>, angularTarget=<1 0 0>}
      localJointOffset=<2 0 0>
     }]
    relativeSpaceBone {boneName=root}
   }
   n20 = Anim Dynamics - Bone: dyn_hair_E_02_r(ComponentPose=@n19, Alpha=1, ExternalForce="0,0,0") {
    angularDampingOverride=0.9999
    angularSpringConstant=100
    bAngularSpring=true
    bGravityOverrideInSimSpace=true
    bOverrideAngularDamping=true
    simulationSpace=RootRelative
    alphaBoolBlend {blendOption=Linear}
    boundBone {boneName=dyn_hair_E_02_r}
    componentAppliedLinearAccClamp=<500 500 500>
    componentLinearAccScale=<0.5 0.5 0.5>
    componentLinearVelScale=<0.5 0.5 0.5>
    physicsBodyDefinitions [{
      sphereCollisionRadius=0
      boundBone {boneName=dyn_hair_E_02_r}
      boxExtents=<15 10 10>
      constraintSetup {angularLimitsMax=<1 0 15>, angularLimitsMin=<-1 -30 -15>, angularTarget=<1 0 0>}
      localJointOffset=<2 0 0>
     }]
    relativeSpaceBone {boneName=root}
   }
   n21 = Anim Dynamics - Bone: dyn_hair_E_01_l(ComponentPose=@n20, Alpha=1, ExternalForce="0,0,0") {
    angularDampingOverride=0.9999
    angularSpringConstant=100
    bAngularSpring=true
    bGravityOverrideInSimSpace=true
    bOverrideAngularDamping=true
    simulationSpace=Actor
    alphaBoolBlend {blendOption=Linear}
    boundBone {boneName=dyn_hair_E_01_l}
    componentAppliedLinearAccClamp=<500 500 500>
    componentLinearAccScale=<0.5 0.5 0.5>
    componentLinearVelScale=<0.5 0.5 0.5>
    physicsBodyDefinitions [{
      sphereCollisionRadius=0
      boundBone {boneName=dyn_hair_E_01_l}
      boxExtents=<15 10 10>
      constraintSetup {angularLimitsMax=<1 10 15>, angularLimitsMin=<-1 -30 -15>, angularTarget=<1 0 0>}
      localJointOffset=<2 0 0>
     }]
    relativeSpaceBone {boneName=root}
   }
   Save cached pose 'animDynamics_Cache'(Pose=Component To Local(ComponentPose=@n21)) {cacheName=animDynamics_Cache}
  ]}
 EventGraph {nodes [
   Event BlueprintUpdateAnimation()
   > n1 = Gravity Override State Function DEPRECATED(
    CharacterPartAnimInstance=Self-Reference()
    GameTimeInSeconds=GetGameTimeInSeconds()
    GravityOverrideParams=Get GravityOverrideParams() -> (GravityOverrideParams:Gravity Override Params Struct Structure)
   )
   > Set GravityOverride(GravityOverride=@n1)
   Event BlueprintInitializeAnimation()
   > Set Skel_Mesh(Skel_Mesh=GetOwnerSkeletalMeshComponent())
  ]}
]
variables [
 GravityOverride {type=Vector}
 GravityOverrideParams {type=GravityOverrideParamsStruct}
 Skel_Mesh {type=SkeletalMeshComponent}
]
